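; Base pawn component that gates, initializes and activates a Fortnite movement mode extension on its owning pawn.
; BeginPlay optionally calls TryInitialize, which checks an enabled ScalableFloat, then routes locally controlled owners to an empty InitializeLocallyControlledClient hook.
; ActivateMovementMode activates the registered extension class unless its gameplay tag is already pushed on the character movement component.
; InitializeInput adds an Enhanced Input mapping context but is not called here.
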(event EventBeginPlay
  (if (Variables|Default|GetbInitializeOnBeginPlay)
    (CallFunction|TryInitialize)))

(event EventTick
  (params (DeltaSeconds :type Float%20%28single-precision%29)))

(fn ActivateMovementMode
  (if (not (CallFunction|IsMovementExtAlreadyActive (Variables|Default|GetMovementModeTag)))
    (bind _returnvalue (Class|FortMovementModeExtBlueprintLibrary|ActivateRegisteredMovementModeExtension (Variables|MovementMode|GetMovementModeExtensionClass) (Components|GetOwner)))
    (return _returnvalue)))

(fn GetMovementComponent
  (return (Actor|GetComponentbyClass (Components|GetOwner) "/Script/FortniteGame.FortMovementComp_Character")))

(fn GetOwnerPawn
  (bind _asfort_player_pawn (Utilities|Casting|CastToFortPlayerPawn (Components|GetOwner)))
  (return _asfort_player_pawn))

(fn Initialize
  (Utilities|SwitchOwnerLocallyControlled
    (:true
      (CallFunction|InitializeLocallyControlledClient))))

(fn InitializeInput
  (Input|AddMappingContext (PlayerController|LocalPlayerSubsystems|GetEnhancedInputLocalPlayerSubsystem (Fortnite|GetFortPlayerControllerfromActor (Components|GetOwner))) (Variables|Input|GetInputMappingContext) 0 "(bIgnoreAllPressedKeysUntilRelease=True,bForceImmediately=False,bNotifyUserSettings=False)"))

(fn InitializeLocallyControlledClient)

(fn IsMovementExtAlreadyActive
  (params (ExtensionTag :type Gameplay%20Tag%20Structure))
  (bind _asfort_player_pawn (Utilities|Casting|CastToFortPlayerPawn (Components|GetOwner)))
  (bind _asfort_movement_comp_character (Utilities|Casting|CastToFortMovementComp_Character (Variables|/Script/Engine.Character|GetCharacterMovement _asfort_player_pawn)))
  (return (Class|FortMovementCompCharacter|IsExtensionPushedwithTag _asfort_movement_comp_character (GameplayTags|MakeGameplayTagContainerfromTag ExtensionTag))))

(fn IsMovementModeEnabled
  (return (ScalableFloat|GetValueasBool (Variables|Data|GetRow_MovementModeEnabled))))

(fn IsOwnerLocallyControlled
  (return (Utilities|Enum|Equal(Enum) (Networking|GetLocalRole (Components|GetOwner)) "ROLE_AutonomousProxy")))

(fn TryInitialize
  (if (CallFunction|IsMovementModeEnabled)
    (CallFunction|Initialize)))
